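# Sound cue for the fist impact sound in the AdHocSquad emote.
# A Random node picks among six impact wave variations with equal weights and feeds a root Modulator for playback variation.
# The cue plays boosted in volume under the player foley sound class, using the shared Athena emote attenuation.
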SoundCue /BRCosmetics/Sounds/Emotes/AdHocSquad/Emote_AdHocSquad_Fist_Impact_Cue
root=Modulator_0
nodes [
 {id=Modulator_0, children [Random_0]}
 {id=Random_0, children [WavePlayer_0,WavePlayer_1,WavePlayer_2,WavePlayer_3,WavePlayer_8,WavePlayer_9], properties {weights [1,1,1,1,1,1]}}
 {id=WavePlayer_0, properties {soundWaveAssetPtr=&/Emote_AdHocSquad_Fist_Impact_06}}
 {id=WavePlayer_1, properties {soundWaveAssetPtr=&/Emote_AdHocSquad_Fist_Impact_05}}
 {id=WavePlayer_2, properties {soundWaveAssetPtr=&/Emote_AdHocSquad_Fist_Impact_04}}
 {id=WavePlayer_3, properties {soundWaveAssetPtr=&/Emote_AdHocSquad_Fist_Impact_03}}
 {id=WavePlayer_8, properties {soundWaveAssetPtr=&/Emote_AdHocSquad_Fist_Impact_02}}
 {id=WavePlayer_9, properties {soundWaveAssetPtr=&/Emote_AdHocSquad_Fist_Impact_01}}
]
properties {
 attenuationSettings=&Fort_Emotes_Athena_Att
 maxDistance=1600
 soundClassObject=&Fort_Player_Foley
 volumeMultiplier=2.5137
}
