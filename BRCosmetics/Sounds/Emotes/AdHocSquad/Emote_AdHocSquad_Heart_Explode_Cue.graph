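# Sound cue for the heart explosion sound in the AdHocSquad emote.
# A Random node picks one of four equally weighted wave variations.
# Output plays at slightly reduced volume in the player foley sound class, spatialized through the shared Athena emote attenuation.
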SoundCue /BRCosmetics/Sounds/Emotes/AdHocSquad/Emote_AdHocSquad_Heart_Explode_Cue
root=Random_0
nodes [
 {id=Random_0, children [WavePlayer_0,WavePlayer_1,WavePlayer_2,WavePlayer_3], properties {weights [1,1,1,1]}}
 {id=WavePlayer_0, properties {soundWaveAssetPtr=&/Emote_AdHocSquad_Heart_Explode_01}}
 {id=WavePlayer_1, properties {soundWaveAssetPtr=&/Emote_AdHocSquad_Heart_Explode_02}}
 {id=WavePlayer_2, properties {soundWaveAssetPtr=&/Emote_AdHocSquad_Heart_Explode_03}}
 {id=WavePlayer_3, properties {soundWaveAssetPtr=&/Emote_AdHocSquad_Heart_Explode_04}}
]
properties {
 attenuationSettings=&Fort_Emotes_Athena_Att
 maxDistance=1600
 soundClassObject=&Fort_Player_Foley
 volumeMultiplier=0.942
}
